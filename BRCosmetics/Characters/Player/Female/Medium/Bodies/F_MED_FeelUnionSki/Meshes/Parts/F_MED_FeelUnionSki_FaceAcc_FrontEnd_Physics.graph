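# Front-end Physics Asset for the face accessory part of the female medium FeelUnionSki character cosmetic, set up for pigtail physics.
# Two five-link capsule chains hang from small base bodies, with limited swing and locked twist at each joint.
# Spine, arm and back-skirt capsules add body shapes without constraints, apparently as colliders for the pigtails.
# A large zero-length capsule on the floor bone acts as a spherical ground shape.
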PhysicsAsset /BRCosmetics/Characters/Player/Female/Medium/Bodies/F_MED_FeelUnionSki/Meshes/Parts/F_MED_FeelUnionSki_FaceAcc_FrontEnd_Physics
bodies {
 base_pigTail_lf_1 [{shapeType=Capsule, length=0.505, radius=0.505, shapeName=base_pigTail_lf_1_capsule, rotation=<rot 0 -90 0>}]
 base_pigTail_rt_1 [{shapeType=Capsule, length=0.505, radius=0.505, shapeName=base_pigTail_rt_1_capsule, rotation=<rot 0 -90 0>}]
 dyn_pigTail_lf_1 [{shapeType=Capsule, length=11.5329, radius=5, shapeName=dyn_pigTail_lf_1_capsule, center=<5 0 0>, rotation=<rot 90 0 0>}]
 dyn_pigTail_lf_2 [{shapeType=Capsule, length=13.6302, radius=5, shapeName=dyn_pigTail_lf_2_capsule, center=<5 0 2>, rotation=<rot 83.7625 0 0>}]
 dyn_pigTail_lf_3 [{shapeType=Capsule, length=12.3527, radius=5.0531, shapeName=dyn_pigTail_lf_3_capsule, center=<5 0 1>, rotation=<rot 83.8054 -0.0006 -0.0007>}]
 dyn_pigTail_lf_4 [{shapeType=Capsule, length=15.6785, radius=6.2569, shapeName=dyn_pigTail_lf_4_capsule, center=<5 0 0>, rotation=<rot 83.7444 -180 180>}]
 dyn_pigTail_lf_5 [{
   shapeType=Capsule
   length=19.6661
   radius=5.8801
   shapeName=dyn_pigTail_lf_5_capsule
   center=<3 -2 0>
   rotation=<rot 48.7958 -13.9703 10.3324>
  }]
 dyn_pigTail_rt_1 [{shapeType=Capsule, length=11.5329, radius=5, shapeName=dyn_pigTail_rt_1_capsule, center=<5 0 0>, rotation=<rot 90 0 0>}]
 dyn_pigTail_rt_2 [{shapeType=Capsule, length=13.6302, radius=5, shapeName=dyn_pigTail_rt_2_capsule, center=<5 0 2>, rotation=<rot 83.7625 0 0>}]
 dyn_pigTail_rt_3 [{shapeType=Capsule, length=12.3527, radius=5.0531, shapeName=dyn_pigTail_rt_3_capsule, center=<5 0 1>, rotation=<rot 84.1425 0.0007 0.0008>}]
 dyn_pigTail_rt_4 [{shapeType=Capsule, length=15.6785, radius=6.2569, shapeName=dyn_pigTail_rt_4_capsule, center=<5 0 0>, rotation=<rot 83.7445 180 -180>}]
 dyn_pigTail_rt_5 [{shapeType=Capsule, length=18.1667, radius=6.446, shapeName=dyn_pigTail_rt_5_capsule, center=<3 2 0>, rotation=<rot 62.3838 11.2447 -18.6867>}]
 dyn_skirt_back_0 [{shapeType=Capsule, length=20.7735, radius=10.2827, shapeName=dyn_skirt_back_0_capsule, center=<10 0 0>}]
 dyn_skirt_back_1 [{shapeType=Capsule, length=17.2871, radius=4.8494, shapeName=dyn_skirt_back_1_capsule, center=<9.6298 0 0>, rotation=<rot 0 35.2975 0>}]
 dyn_skirt_back_side_l_0 [{shapeType=Capsule, length=20.7735, radius=10.2827, shapeName=dyn_skirt_back_side_l_0_capsule, center=<10 0 0>}]
 dyn_skirt_back_side_l_1 [{shapeType=Capsule, length=9.4687, radius=5.1144, shapeName=dyn_skirt_back_side_l_1_capsule, center=<11.8557 2.4992 0>}]
 dyn_skirt_back_side_r_0 [{shapeType=Capsule, length=20.7735, radius=10.2827, shapeName=dyn_skirt_back_side_r_0_capsule, center=<10 0 0>}]
 dyn_skirt_back_side_r_1 [{shapeType=Capsule, length=9.4687, radius=5.1144, shapeName=dyn_skirt_back_side_r_1_capsule, center=<11.8557 2.4992 0>}]
 floor [{shapeType=Capsule, length=0, radius=215, shapeName=floor_capsule, center=<0 0 -200>, rotation=<rot 0 -90 0>}]
 lowerarm_l [{shapeType=Capsule, length=28.3594, radius=7.5891, shapeName=lowerarm_l_capsule, center=<8 2 0>, rotation=<rot 90 -10 0>}]
 lowerarm_r [{shapeType=Capsule, length=28.3594, radius=7.5891, shapeName=lowerarm_r_capsule, center=<-8 -2 0>, rotation=<rot 90 -10 0>}]
 spine_02 [{shapeType=Capsule, length=19.8844, radius=10.7915, shapeName=clavicle_l_capsule}]
 spine_04 [{shapeType=Capsule, length=20.8628, radius=9.8132, shapeName=clavicle_l_capsule, center=<-2 0 0>}]
 spine_05 [
  {shapeType=Capsule, length=22.335, radius=8.341, shapeName=clavicle_l_capsule}
  {shapeType=Capsule, length=66.9251, radius=15, shapeName=spine_05_capsule, center=<0 8.0107 0>}
 ]
 upperarm_l [{shapeType=Capsule, length=24.6906, radius=6.3624, shapeName=upperarm_l_capsule, center=<10 0 -1.6>, rotation=<rot -85 0 0>}]
 upperarm_r [{shapeType=Capsule, length=24.6906, radius=6.3624, shapeName=upperarm_l_capsule, center=<-10 0 1.6>, rotation=<rot -85 0 0>}]
}
constraints [
 {
  bone1Name=dyn_pigTail_lf_1
  bone2Name=base_pigTail_lf_1
  swing1LimitDegrees=30
  swing2LimitDegrees=20
  twistLimitDegrees=45
  twistMotion=Locked
 }
 {
  bone1Name=dyn_pigTail_lf_2
  bone2Name=dyn_pigTail_lf_1
  swing1LimitDegrees=30
  swing2LimitDegrees=20
  twistLimitDegrees=45
  twistMotion=Locked
 }
 {
  bone1Name=dyn_pigTail_lf_3
  bone2Name=dyn_pigTail_lf_2
  swing1LimitDegrees=30
  swing2LimitDegrees=20
  twistLimitDegrees=45
  twistMotion=Locked
 }
 {
  bone1Name=dyn_pigTail_lf_4
  bone2Name=dyn_pigTail_lf_3
  swing1LimitDegrees=30
  swing2LimitDegrees=20
  twistLimitDegrees=45
  twistMotion=Locked
 }
 {
  bone1Name=dyn_pigTail_lf_5
  bone2Name=dyn_pigTail_lf_4
  swing1LimitDegrees=30
  swing2LimitDegrees=20
  twistLimitDegrees=45
  twistMotion=Locked
 }
 {
  bone1Name=dyn_pigTail_rt_1
  bone2Name=base_pigTail_rt_1
  swing1LimitDegrees=30
  swing2LimitDegrees=20
  twistLimitDegrees=45
  twistMotion=Locked
 }
 {
  bone1Name=dyn_pigTail_rt_2
  bone2Name=dyn_pigTail_rt_1
  swing1LimitDegrees=30
  swing2LimitDegrees=20
  twistLimitDegrees=45
  twistMotion=Locked
 }
 {
  bone1Name=dyn_pigTail_rt_3
  bone2Name=dyn_pigTail_rt_2
  swing1LimitDegrees=30
  swing2LimitDegrees=20
  twistLimitDegrees=45
  twistMotion=Locked
 }
 {
  bone1Name=dyn_pigTail_rt_4
  bone2Name=dyn_pigTail_rt_3
  swing1LimitDegrees=30
  swing2LimitDegrees=20
  twistLimitDegrees=45
  twistMotion=Locked
 }
 {
  bone1Name=dyn_pigTail_rt_5
  bone2Name=dyn_pigTail_rt_4
  swing1LimitDegrees=30
  swing2LimitDegrees=20
  twistLimitDegrees=45
  twistMotion=Locked
 }
]
