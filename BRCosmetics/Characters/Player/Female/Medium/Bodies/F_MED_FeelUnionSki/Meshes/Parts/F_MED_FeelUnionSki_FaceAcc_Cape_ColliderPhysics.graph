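# Physics asset for the pigtail bones of the female medium FeelUnionSki cosmetic's face accessory, named as a cape collider.
# Left and right pigtails are each a five-bone chain with one capsule per bone, closely mirrored except at the tips, where size and angle differ.
# Constraints link each bone to the previous one with locked twist and identical swing angle settings, with none joining the two chains or attaching their roots.
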PhysicsAsset /BRCosmetics/Characters/Player/Female/Medium/Bodies/F_MED_FeelUnionSki/Meshes/Parts/F_MED_FeelUnionSki_FaceAcc_Cape_ColliderPhysics
bodies {
 dyn_pigTail_lf_1 [{shapeType=Capsule, length=15.6012, radius=5.7537, shapeName=dyn_pigTail_lf_1_capsule, center=<5 0 0>, rotation=<rot 90 0 0>}]
 dyn_pigTail_lf_2 [{shapeType=Capsule, length=17.6986, radius=5.7537, shapeName=dyn_pigTail_lf_2_capsule, center=<5 0 2>, rotation=<rot 83.7625 0 0>}]
 dyn_pigTail_lf_3 [{shapeType=Capsule, length=16.4211, radius=5.8068, shapeName=dyn_pigTail_lf_3_capsule, center=<5 0 1>, rotation=<rot 78.0451 -0.0002 -0.0003>}]
 dyn_pigTail_lf_4 [{shapeType=Capsule, length=19.7468, radius=7.0106, shapeName=dyn_pigTail_lf_4_capsule, center=<5 0 0>, rotation=<rot 83.7444 -180 180>}]
 dyn_pigTail_lf_5 [{
   shapeType=Capsule
   length=15.6108
   radius=8.3831
   shapeName=dyn_pigTail_lf_5_capsule
   center=<3 -2 0>
   rotation=<rot 48.7958 -13.9703 10.3324>
  }]
 dyn_pigTail_rt_1 [{shapeType=Capsule, length=15.6012, radius=5.7537, shapeName=dyn_pigTail_rt_1_capsule, center=<5 0 0>, rotation=<rot 90 0 0>}]
 dyn_pigTail_rt_2 [{shapeType=Capsule, length=17.6986, radius=5.7537, shapeName=dyn_pigTail_rt_2_capsule, center=<5 0 2>, rotation=<rot 83.7625 0 0>}]
 dyn_pigTail_rt_3 [{shapeType=Capsule, length=16.4211, radius=5.8068, shapeName=dyn_pigTail_rt_3_capsule, center=<5 0 1>, rotation=<rot 78.0451 0.0002 0.0003>}]
 dyn_pigTail_rt_4 [{shapeType=Capsule, length=19.7468, radius=7.0106, shapeName=dyn_pigTail_rt_4_capsule, center=<5 0 0>, rotation=<rot 83.7445 180 -180>}]
 dyn_pigTail_rt_5 [{
   shapeType=Capsule
   length=14.1114
   radius=8.9489
   shapeName=dyn_pigTail_rt_5_capsule
   center=<3 2 0>
   rotation=<rot 62.3838 11.2447 -18.6867>
  }]
}
constraints [
 {
  bone1Name=dyn_pigTail_lf_2
  bone2Name=dyn_pigTail_lf_1
  swing1LimitDegrees=15
  swing2LimitDegrees=20
  twistLimitDegrees=45
  twistMotion=Locked
 }
 {
  bone1Name=dyn_pigTail_lf_3
  bone2Name=dyn_pigTail_lf_2
  swing1LimitDegrees=15
  swing2LimitDegrees=20
  twistLimitDegrees=45
  twistMotion=Locked
 }
 {
  bone1Name=dyn_pigTail_lf_4
  bone2Name=dyn_pigTail_lf_3
  swing1LimitDegrees=15
  swing2LimitDegrees=20
  twistLimitDegrees=45
  twistMotion=Locked
 }
 {
  bone1Name=dyn_pigTail_lf_5
  bone2Name=dyn_pigTail_lf_4
  swing1LimitDegrees=15
  swing2LimitDegrees=20
  twistLimitDegrees=45
  twistMotion=Locked
 }
 {
  bone1Name=dyn_pigTail_rt_2
  bone2Name=dyn_pigTail_rt_1
  swing1LimitDegrees=15
  swing2LimitDegrees=20
  twistLimitDegrees=45
  twistMotion=Locked
 }
 {
  bone1Name=dyn_pigTail_rt_3
  bone2Name=dyn_pigTail_rt_2
  swing1LimitDegrees=15
  swing2LimitDegrees=20
  twistLimitDegrees=45
  twistMotion=Locked
 }
 {
  bone1Name=dyn_pigTail_rt_4
  bone2Name=dyn_pigTail_rt_3
  swing1LimitDegrees=15
  swing2LimitDegrees=20
  twistLimitDegrees=45
  twistMotion=Locked
 }
 {
  bone1Name=dyn_pigTail_rt_5
  bone2Name=dyn_pigTail_rt_4
  swing1LimitDegrees=15
  swing2LimitDegrees=20
  twistLimitDegrees=45
  twistMotion=Locked
 }
]
